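; Gameplay ability for the Creative flashlight ranged weapon that toggles its spotlight on or off when activated.
; It derives the new state by inverting the current spotlight visibility on the player's equipped weapon, then stores and applies it.
; It executes a toggle gameplay cue with the state encoded as normalized magnitude.
; Only the server sends the matching on or off gameplay event, then the ability ends.
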(event Ability|EventActivateAbility
  (|CallParentFunction)
  (bind _asfort_player_pawn (Utilities|Casting|CastToFortPlayerPawn (Ability|GetAvatarActorfromActorInfo)))
  (bind _asb_ranged_flashlight (Utilities|Casting|CastToB_Ranged_Flashlight (Weapon|GetCurrentWeapon _asfort_player_pawn)))
  (bind _output_get (Variables|State|SetbToggleOn (not (Rendering|IsVisible (Variables|/Game/Athena/Items/Weapons/Creative/Ranged/B_Ranged_Flashlight.B_Ranged_Flashlight_C|GetSpotLight _asb_ranged_flashlight)))))
  (Rendering|SetVisibility (Variables|/Game/Athena/Items/Weapons/Creative/Ranged/B_Ranged_Flashlight.B_Ranged_Flashlight_C|GetSpotLight _asb_ranged_flashlight) _output_get false)
  (Ability|ExecuteGameplayCueWithParamsOnOwner (Variables|FX|GetCUETAG_FlashLightToggled) (Ability|GameplayCue|MakeGameplayCueParameters :NormalizedMagnitude (Math|Float|SelectFloat 1.0 0.0 (Variables|State|GetbToggleOn)) :RawMagnitude 0.0 :Location "0,0,0" :Normal "0,0,0" :Instigator (Ability|GetCurrentSourceWeapon) :EffectCauser (Ability|GetAvatarActorfromActorInfo) :GameplayEffectLevel 1 :AbilityLevel 1 :bReplicateLocationWhenUsingMinimalRepProxy false))
  (Utilities|IsServer
    (:true
      (Ability|SendGameplayEvent (select (Variables|State|GetbToggleOn) (Variables|State|GetGTAG_ToggleOn) (Variables|State|GetGTAG_ToggleOff))))
    (:false
      _))
  (Ability|EndAbility))
