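; Gameplay ability for the Creative flashlight weapon that tracks actors caught in its beam and manages their lit state.
; The Toggle.On event sets a 0.1-second looping timer running a "LightingSweep" target selection.
; Each sweep applies an effect container to current hits and removes GE_Flashlight_IsLit, sourced from the owner, from actors that dropped out.
; Ending the ability clears the timer and remaining lit targets.
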(event Custom|TickBeam
  (Utilities|Array|Clear (Variables|Default|GetNewHitTargets))
  (bind (_asynctaskproxy _targetdata _applicationtag) (Ability|Tasks|DoTargetSelection "LightingSweep")
    (:Targeted
      (Targeting|GetAllHitActors _targetdata (Variables|Default|GetNewHitTargets)
        (:Then
          (for _array_element (Variables|Default|GetOldHitTargets)
            (if (< (Utilities|Array|FindItem (Variables|Default|GetNewHitTargets) _array_element) 0)
              (CallFunction|ClearLitState _array_element)))
          (Ability|ApplyGameplayEffectContainer _targetdata _applicationtag)
          (Utilities|Array|Clear (Variables|Default|GetOldHitTargets))
          (Variables|Default|SetOldHitTargets (Variables|Default|GetNewHitTargets)))))))

(event Ability|EventActivateAbilityFromEvent
  (params (EventData :type Gameplay%20Event%20Data%20Structure%20%28by%20ref%29))
  (|CallParentFunction EventData)
  (switch Utilities|FlowControl|Switch|SwitchonGameplayTag (Utilities|Struct|BreakGameplayEventData EventData)
    (:Default
      _)
    (:Event.Weapons.Flashlight.Toggle.On
      (bind _returnvalue (Utilities|Time|SetTimerbyEvent (delegate Custom|TickBeam) 0.1 true))
      (Variables|Default|SetTimerTick _returnvalue))
    (:Event.Weapons.Flashlight.Toggle.Off
      _))
  (Ability|EndAbility))

(event Ability|EventOnEndAbility
  (params (bWasCancelled :type Boolean))
  (Utilities|Time|ClearandInvalidateTimerbyHandle (Variables|Default|GetTimerTick))
  (for _array_element (Variables|Default|GetOldHitTargets)
    (CallFunction|ClearLitState _array_element))
  (Utilities|Array|Clear (Variables|Default|GetOldHitTargets)))

(fn ClearLitState
  (params (Actor :type Actor%20Object%20Reference))
  (GameplayEffects|RemoveActiveGameplayEffectbySourceEffect (Ability|GetAbilitySystemComponent Actor) "/Game/Athena/Items/Weapons/Creative/Ranged/GE_Flashlight_IsLit.GE_Flashlight_IsLit_C" (Ability|GetAbilitySystemComponent (Variables|Default|GetAbilityOwner)) -1))
